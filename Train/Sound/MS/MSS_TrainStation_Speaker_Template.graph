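# Stereo looping MetaSound Source template for a train station PA speaker, playing announcement bus audio as a spatially attenuated source.
# An Audio Bus Reader feeds both channels of the exposed Audio Bus input, defaulting to the station announcement bus, straight to the outputs.
# The graph has no wave player or processing, so the sound depends on whatever is sent to the bus.
# A small PA announcement attenuation asset shapes falloff.
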MetaSoundSource /Train/Sound/MS/MSS_TrainStation_Speaker_Template
inputs [
 Audio Bus {type=AudioBusAsset}
 UE.Source.OnPlay {type=Trigger}
]
interfaces [UE.OutputFormat.Stereo:1.0, UE.Source:1.1]
outputs [
 "UE.OutputFormat.Stereo.Audio:0" {type=Audio}
 "UE.OutputFormat.Stereo.Audio:1" {type=Audio}
]
pages [Default {
  input_defaults {Audio Bus=/Train/Sound/AB_TrainStationAnnouncmentBus, UE.Source.OnPlay=null}
  nodes [AudioBusReader2_0 = "UE.Audio Bus Reader (2)"(Audio Bus=^Audio Bus)]
  links [{dst="@UE.OutputFormat.Stereo.Audio:0", src=AudioBusReader2_0.Out 0}, {dst="@UE.OutputFormat.Stereo.Audio:1", src=AudioBusReader2_0.Out 1}]
 }]
properties {
 assetImportData=&self:AssetImportData
 attenuationSettings=/Train/Sound/Settings/Train_Station_PA_Announcement_Small_Att
 bLooping=true
 duration=10000
 maxDistance=7000
 numChannels=2
 outputFormat=Stereo
 sampleRate=48000
}
