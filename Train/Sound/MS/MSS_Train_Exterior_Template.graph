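# Stereo MetaSound source template for train exterior sounds, playing a random wave from Variations, with looping ones starting at a random time.
# Smoothed Movement.Speed, scaled to MaxTrainSpeed, drives volume and, when VelocityMod is on, pitch.
# Modulation buses for enclosure and player-on-train also scale gain, and enclosure sets a low-pass cutoff.
# A separate Mobile page keeps the same logic but drops the low-pass filters.
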MetaSoundSource /Train/Sound/MS/MSS_Train_Exterior_Template
inputs [
 Enclosed.LPF {type=Float}
 Enclosed.Volume {type=Float}
 Gain.PlayerOffTrain {type=Float}
 Gain.PlayerOnTrain {type=Float}
 Loop {type=Bool}
 MaxTrainSpeed {type=Float}
 Movement.Speed {type=Float}
 Movement.SpeedAzimuth {type=Float}
 Movement.SpeedElevation {type=Float}
 Pitch {type=Float}
 UE.Source.OnPlay {type=Trigger}
 Variations {type=WaveAsset:Array}
 VelocityMod {type=Bool}
 VolumeMaxSpeed {type=Float}
 VolumeMinSpeed {type=Float}
]
interfaces [Fort.Movement:1.0, UE.OutputFormat.Stereo:1.0, UE.Source.OneShot:1.0, UE.Source:1.1]
outputs [
 "UE.OutputFormat.Stereo.Audio:0" {type=Audio}
 "UE.OutputFormat.Stereo.Audio:1" {type=Audio}
 UE.Source.OneShot.OnFinished {type=Trigger}
]
pages [
 Default {
  is_build_graph=true
  input_defaults {
   Enclosed.LPF=6000
   Enclosed.Volume=0.1
   Gain.PlayerOffTrain=1
   Gain.PlayerOnTrain=1
   Loop="<MetaSoundBuilderResult.SUCCEEDED: 0>"
   MaxTrainSpeed=1000
   Movement.Speed=0
   Movement.SpeedAzimuth=0
   Movement.SpeedElevation=0
   Pitch=1
   UE.Source.OnPlay=null
   VelocityMod="<MetaSoundBuilderResult.SUCCEEDED: 0>"
   VolumeMaxSpeed=1
   VolumeMinSpeed=0
   Variations []
  }
  nodes [
   n1 = Modulation.GetModulatorValue() {also {defaults {Modulator=&AudioWorldization_Bus_Enclosure}}}
   n2 = Clamp.Clamp:Float(In=MapRange.MapRange:Float(Out Range B=^Enclosed.LPF, In=@n1.Out) {also {defaults {Out Range A=20000}}}.Out Value) {also {defaults {Max=20000, Min=20}}}
   n3 = Modulation.GetModulatorValue() {also {defaults {Modulator=/Train/Sound/Settings/CB_PlayerIsOnTrain}}}
   n4 = InitVariable.Float()
   n5 = InitVariable.Float()
   TriggerOnValueChange_1 = UE.Trigger On Value Change:Float(Value=^Movement.Speed)
   n6 = Value.Value:Float(
    Target Value=^Movement.Speed
    Set=TriggerCompare.Clamp:Float(A=^Movement.Speed, Compare=@TriggerOnValueChange_1.Trigger) {also {defaults {B=0, Type=1}}}.True
   ) {also {defaults {Init Value=0}}}
   n7 = UE.InterpTo:Audio(Target=VariableAccessor.Float(Variable=VariableMutator.Float(Variable=@n5.Variable, Value=@n6.Output Value).Variable).Value) {also {defaults {Interp Time=0.5}}}
   n8 = MapRange.MapRange:Float(In Range B=^MaxTrainSpeed, Out Range B=^VolumeMaxSpeed, Out Range A=^VolumeMinSpeed, In=@n7.Value)
   n9 = UE.Multiply:Float(
    AdditionalOperands=UE.Multiply:Float(
     AdditionalOperands=MapRange.MapRange:Float(Out Range A=^Gain.PlayerOffTrain, Out Range B=^Gain.PlayerOnTrain, In=@n3.Out).Out Value
     PrimaryOperand=MapRange.MapRange:Float(Out Range B=^Enclosed.Volume, In=@n1.Out) {also {defaults {Out Range A=1}}}.Out Value
    ).out
    PrimaryOperand=VariableAccessor.Float(Variable=VariableMutator.Float(Variable=@n4.Variable, Value=@n8.Out Value).Variable).Value
   )
   ArrayRandomGet_0 = Array.Random Get:WaveAsset:Array(Next=^UE.Source.OnPlay, In Array=^Variations)
   GetWaveDuration_0 = "UE.Get Wave Duration: "(Wave=@ArrayRandomGet_0.Value)
   TriggerOnValueChange_0 = UE.Trigger On Value Change:Float(Value=UE.ConversionTimeToFloat(In=@GetWaveDuration_0.Duration).Out)
   n10 = TriggerCompare.Clamp:Bool(A=^Loop, Compare=@ArrayRandomGet_0.On Next) {also {defaults {B="<MetaSoundBuilderResult.SUCCEEDED: 0>"}}}
   TriggerRouteTriggerRouteTime2_0 = "TriggerRoute.Trigger Route (Time, 2):Time"(Value 0=UE.RandomTime(Max=@GetWaveDuration_0.Duration, Next=@TriggerOnValueChange_0.Trigger).Value, Set 1=@n10.False, Set 0=@n10.True)
   n11 = UE.Subtract:Float(AdditionalOperands=Convert.Float:Bool(Bool=^VelocityMod).Float) {also {defaults {PrimaryOperand=1}}}
   n12 = Convert.Float:Bool(Bool=^VelocityMod)
   n13 = UE.Multiply:Float(
    PrimaryOperand=^Pitch
    AdditionalOperands=MapRange.MapRange:Float(In Range B=^MaxTrainSpeed, In=@n7.Value) {also {defaults {Out Range A=0.5}}}.Out Value
   )
   FrequencyMultipliertoSemitone_0 = UE.Frequency Multiplier to Semitone:Float(Frequency Multiplier=UE.Add:Float(
     PrimaryOperand=UE.Multiply:Float(AdditionalOperands=^Pitch, PrimaryOperand=@n11.Out).out
     AdditionalOperands=UE.Multiply:Float(AdditionalOperands=@n12.Float, PrimaryOperand=@n13.out).out
    ).Out)
   n14 = InitVariable.Float()
   WavePlayer_0 = UE.Wave Player:Stereo(
    Loop=^Loop
    Wave Asset=@ArrayRandomGet_0.Value
    Play=@TriggerRouteTriggerRouteTime2_0.On Set
    Start Time=@TriggerRouteTriggerRouteTime2_0.Value
    Pitch Shift=VariableAccessor.Float(Variable=VariableMutator.Float(Value=@FrequencyMultipliertoSemitone_0.Semitones, Variable=@n14.Variable).Variable).Value
   ) {also {version [1,1]}}
   Multiply_1 = UE.Multiply:Audio by Float(AdditionalOperands=@n9.out, PrimaryOperand=@WavePlayer_0.Out Right)
   OnePoleLowPassFilter_0 = UE.One-Pole Low Pass Filter:Audio(Cutoff Frequency=@n2.Value, In=@Multiply_1.Out)
   Multiply_0 = UE.Multiply:Audio by Float(AdditionalOperands=@n9.out, PrimaryOperand=@WavePlayer_0.Out Left)
   OnePoleLowPassFilter_1 = UE.One-Pole Low Pass Filter:Audio(Cutoff Frequency=@n2.Value, In=@Multiply_0.Out)
  ]
  links [
   {dst="@UE.OutputFormat.Stereo.Audio:1", src=OnePoleLowPassFilter_0.Out}
   {dst="@UE.OutputFormat.Stereo.Audio:0", src=OnePoleLowPassFilter_1.Out}
   {dst="@UE.Source.OneShot.OnFinished", src=WavePlayer_0.On Finished}
  ]
 }
 Mobile {
  nodes [
   n1 = Modulation.GetModulatorValue() {also {defaults {Modulator=&CB_PlayerIsOnTrain}}}
   n2 = Modulation.GetModulatorValue() {also {defaults {Modulator=&AudioWorldization_Bus_Enclosure}}}
   n3 = InitVariable.Float()
   n4 = InitVariable.Float()
   TriggerOnValueChange_1 = UE.Trigger On Value Change:Float(Value=^Movement.Speed)
   n5 = Value.Value:Float(
    Target Value=^Movement.Speed
    Set=TriggerCompare.Clamp:Float(A=^Movement.Speed, Compare=@TriggerOnValueChange_1.Trigger) {also {defaults {B=0, Type=1}}}.True
   ) {also {defaults {Init Value=0}}}
   n6 = UE.InterpTo:Audio(Target=VariableAccessor.Float(Variable=VariableMutator.Float(Variable=@n4.Variable, Value=@n5.Output Value).Variable).Value) {also {defaults {Interp Time=0.5}}}
   n7 = MapRange.MapRange:Float(In Range B=^MaxTrainSpeed, Out Range B=^VolumeMaxSpeed, Out Range A=^VolumeMinSpeed, In=@n6.Value)
   n8 = UE.Multiply:Float(
    AdditionalOperands=UE.Multiply:Float(
     AdditionalOperands=MapRange.MapRange:Float(Out Range A=^Gain.PlayerOffTrain, Out Range B=^Gain.PlayerOnTrain, In=@n1.Out).Out Value
     PrimaryOperand=MapRange.MapRange:Float(Out Range B=^Enclosed.Volume, In=@n2.Out) {also {defaults {Out Range A=1}}}.Out Value
    ).out
    PrimaryOperand=VariableAccessor.Float(Variable=VariableMutator.Float(Variable=@n3.Variable, Value=@n7.Out Value).Variable).Value
   )
   ArrayRandomGet_0 = Array.Random Get:WaveAsset:Array(Next=^UE.Source.OnPlay, In Array=^Variations)
   GetWaveDuration_0 = "UE.Get Wave Duration: "(Wave=@ArrayRandomGet_0.Value)
   TriggerOnValueChange_0 = UE.Trigger On Value Change:Float(Value=UE.ConversionTimeToFloat(In=@GetWaveDuration_0.Duration).Out)
   n9 = TriggerCompare.Clamp:Bool(A=^Loop, Compare=@ArrayRandomGet_0.On Next) {also {defaults {B="<MetaSoundBuilderResult.SUCCEEDED: 0>"}}}
   TriggerRouteTriggerRouteTime2_0 = "TriggerRoute.Trigger Route (Time, 2):Time"(Value 0=UE.RandomTime(Max=@GetWaveDuration_0.Duration, Next=@TriggerOnValueChange_0.Trigger).Value, Set 1=@n9.False, Set 0=@n9.True)
   n10 = UE.Subtract:Float(AdditionalOperands=Convert.Float:Bool(Bool=^VelocityMod).Float) {also {defaults {PrimaryOperand=1}}}
   n11 = Convert.Float:Bool(Bool=^VelocityMod)
   n12 = UE.Multiply:Float(
    PrimaryOperand=^Pitch
    AdditionalOperands=MapRange.MapRange:Float(In Range B=^MaxTrainSpeed, In=@n6.Value) {also {defaults {Out Range A=0.5}}}.Out Value
   )
   FrequencyMultipliertoSemitone_0 = UE.Frequency Multiplier to Semitone:Float(Frequency Multiplier=UE.Add:Float(
     PrimaryOperand=UE.Multiply:Float(AdditionalOperands=^Pitch, PrimaryOperand=@n10.Out).out
     AdditionalOperands=UE.Multiply:Float(AdditionalOperands=@n11.Float, PrimaryOperand=@n12.out).out
    ).Out)
   n13 = InitVariable.Float()
   WavePlayer_0 = UE.Wave Player:Stereo(
    Loop=^Loop
    Wave Asset=@ArrayRandomGet_0.Value
    Play=@TriggerRouteTriggerRouteTime2_0.On Set
    Start Time=@TriggerRouteTriggerRouteTime2_0.Value
    Pitch Shift=VariableAccessor.Float(Variable=VariableMutator.Float(Value=@FrequencyMultipliertoSemitone_0.Semitones, Variable=@n13.Variable).Variable).Value
   ) {also {version [1,1]}}
   Multiply_0 = UE.Multiply:Audio by Float(AdditionalOperands=@n8.out, PrimaryOperand=@WavePlayer_0.Out Left)
   Multiply_1 = UE.Multiply:Audio by Float(AdditionalOperands=@n8.out, PrimaryOperand=@WavePlayer_0.Out Right)
  ]
  links [
   {dst="@UE.OutputFormat.Stereo.Audio:0", src=Multiply_0.Out}
   {dst="@UE.OutputFormat.Stereo.Audio:1", src=Multiply_1.Out}
   {dst="@UE.Source.OneShot.OnFinished", src=WavePlayer_0.On Finished}
  ]
 }
]
properties {assetImportData=&self:AssetImportData, numChannels=2, outputFormat=Stereo, sampleRate=48000}
