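# Quad MetaSound source for looping train rooftop wind, built from front and rear snowstorm loops with a shared pitch shift.
# Front outputs mix both loops; the other pair carries only the attenuated rear loop, with left and right crossed.
# An ADSR envelope fades both in on play and out on Stop, halting the players when done.
# The Default page randomizes loop start times from wave duration; the Mobile page omits this.
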MetaSoundSource /Train/Sound/MS/MSS_Train_Rooftop_Wind
inputs [
 FrontLoop {type=WaveAsset}
 Pitch Shift {type=Float}
 RearLoop {type=WaveAsset}
 Stop {type=Trigger}
 UE.Source.OnPlay {type=Trigger}
]
interfaces [UE.OutputFormat.Quad:1.0, UE.Source:1.1]
outputs [
 "UE.OutputFormat.Quad.Audio:0" {type=Audio}
 "UE.OutputFormat.Quad.Audio:1" {type=Audio}
 "UE.OutputFormat.Quad.Audio:4" {type=Audio}
 "UE.OutputFormat.Quad.Audio:5" {type=Audio}
]
pages [
 Default {
  is_build_graph=true
  input_defaults {
   FrontLoop=/Game/Sounds/Apollo/Weather/SnowStorm/Weather_SnowStorm_Accumulating_Heavy_Loop_01
   Pitch Shift=0
   RearLoop=/Game/Sounds/Apollo/Weather/SnowStorm/Weather_SnowStorm_Buffeting_Heavy_Loop
   Stop=null
   UE.Source.OnPlay=null
  }
  nodes [
   ADSREnvelopeADSREnvelope_0 = ADSR Envelope.ADSR Envelope:Float(Trigger Release=^Stop, Trigger Attack=^UE.Source.OnPlay) {also {defaults {Attack Time=1, Sustain Level=1}}}
   n1 = VariableAccessor.Float(Variable=VariableMutator.Float(Value=@ADSREnvelopeADSREnvelope_0.Out Envelope, Variable=InitVariable.Float().Variable).Variable)
   n2 = VariableAccessor.Float(Variable=@n1.Variable)
   GetWaveDuration_0 = "UE.Get Wave Duration: "(Wave=^RearLoop)
   TriggerOnValueChange_1 = UE.Trigger On Value Change:Float(Value=UE.ConversionTimeToFloat(In=@GetWaveDuration_0.Duration).Out)
   n3 = VariableAccessor.Trigger(Variable=VariableMutator.Trigger(Value=@ADSREnvelopeADSREnvelope_0.On Done, Variable=InitVariable.Trigger().Variable).Variable)
   WavePlayer_0 = UE.Wave Player:Stereo(
    Pitch Shift=^Pitch Shift
    Wave Asset=^RearLoop
    Play=^UE.Source.OnPlay
    Start Time=UE.RandomTime(Max=@GetWaveDuration_0.Duration, Next=@TriggerOnValueChange_1.Trigger).Value
    Stop=VariableAccessor.Trigger(Variable=@n3.Variable).Value
   ) {also {defaults {Loop="<MetaSoundBuilderResult.SUCCEEDED: 0>"}, version [1,1]}}
   GetWaveDuration_1 = "UE.Get Wave Duration: "(Wave=^FrontLoop)
   TriggerOnValueChange_0 = UE.Trigger On Value Change:Float(Value=UE.ConversionTimeToFloat(In=@GetWaveDuration_1.Duration).Out)
   WavePlayer_1 = UE.Wave Player:Stereo(
    Wave Asset=^FrontLoop
    Pitch Shift=^Pitch Shift
    Play=^UE.Source.OnPlay
    Start Time=UE.RandomTime(Max=@GetWaveDuration_1.Duration, Next=@TriggerOnValueChange_0.Trigger).Value
    Stop=@n3.Value
   ) {also {defaults {Loop="<MetaSoundBuilderResult.SUCCEEDED: 0>"}, version [1,1]}}
   AudioMixerAudioMixerStereo2_0 = "AudioMixer.Audio Mixer (Stereo, 2)"(
    Gain 0=@n2.Value
    Gain 1=@n2.Value
    In 1 L=@WavePlayer_0.Out Left
    In 1 R=@WavePlayer_0.Out Right
    In 0 L=@WavePlayer_1.Out Left
    In 0 R=@WavePlayer_1.Out Right
   )
   n4 = UE.Multiply:Float(PrimaryOperand=@n1.Value) {also {defaults {AdditionalOperands=0.75}}}
   Multiply_0 = UE.Multiply:Audio by Float(AdditionalOperands=@n4.out, PrimaryOperand=@WavePlayer_0.Out Left)
   Multiply_1 = UE.Multiply:Audio by Float(AdditionalOperands=@n4.out, PrimaryOperand=@WavePlayer_0.Out Right)
  ]
  links [
   {dst="@UE.OutputFormat.Quad.Audio:0", src=AudioMixerAudioMixerStereo2_0.Out L}
   {dst="@UE.OutputFormat.Quad.Audio:1", src=AudioMixerAudioMixerStereo2_0.Out R}
   {dst="@UE.OutputFormat.Quad.Audio:5", src=Multiply_0.out}
   {dst="@UE.OutputFormat.Quad.Audio:4", src=Multiply_1.out}
  ]
 }
 Mobile {
  nodes [
   ADSREnvelopeADSREnvelope_0 = ADSR Envelope.ADSR Envelope:Float(Trigger Release=^Stop, Trigger Attack=^UE.Source.OnPlay) {also {defaults {Attack Time=1, Sustain Level=1}}}
   n1 = VariableAccessor.Float(Variable=VariableMutator.Float(Value=@ADSREnvelopeADSREnvelope_0.Out Envelope, Variable=InitVariable.Float().Variable).Variable)
   n2 = VariableAccessor.Float(Variable=@n1.Variable)
   n3 = VariableAccessor.Trigger(Variable=VariableMutator.Trigger(Value=@ADSREnvelopeADSREnvelope_0.On Done, Variable=InitVariable.Trigger().Variable).Variable)
   WavePlayer_0 = UE.Wave Player:Stereo(Pitch Shift=^Pitch Shift, Wave Asset=^RearLoop, Play=^UE.Source.OnPlay, Stop=VariableAccessor.Trigger(Variable=@n3.Variable).Value) {also {defaults {Loop="<MetaSoundBuilderResult.SUCCEEDED: 0>"}, version [1,1]}}
   WavePlayer_1 = UE.Wave Player:Stereo(Wave Asset=^FrontLoop, Pitch Shift=^Pitch Shift, Play=^UE.Source.OnPlay, Stop=@n3.Value) {also {defaults {Loop="<MetaSoundBuilderResult.SUCCEEDED: 0>"}, version [1,1]}}
   AudioMixerAudioMixerStereo2_0 = "AudioMixer.Audio Mixer (Stereo, 2)"(
    Gain 0=@n2.Value
    Gain 1=@n2.Value
    In 1 L=@WavePlayer_0.Out Left
    In 1 R=@WavePlayer_0.Out Right
    In 0 L=@WavePlayer_1.Out Left
    In 0 R=@WavePlayer_1.Out Right
   )
   n4 = UE.Multiply:Float(PrimaryOperand=@n1.Value) {also {defaults {AdditionalOperands=0.75}}}
   Multiply_0 = UE.Multiply:Audio by Float(AdditionalOperands=@n4.out, PrimaryOperand=@WavePlayer_0.Out Left)
   Multiply_1 = UE.Multiply:Audio by Float(AdditionalOperands=@n4.out, PrimaryOperand=@WavePlayer_0.Out Right)
  ]
  links [
   {dst="@UE.OutputFormat.Quad.Audio:0", src=AudioMixerAudioMixerStereo2_0.Out L}
   {dst="@UE.OutputFormat.Quad.Audio:1", src=AudioMixerAudioMixerStereo2_0.Out R}
   {dst="@UE.OutputFormat.Quad.Audio:5", src=Multiply_0.out}
   {dst="@UE.OutputFormat.Quad.Audio:4", src=Multiply_1.out}
  ]
 }
]
properties {
 assetImportData=&self:AssetImportData
 bLooping=true
 duration=10000
 numChannels=4
 outputFormat=Quad
 sampleRate=48000
}
